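# Unlit two-sided translucent material for the Caretaker melee air-flow trail, usable on Niagara ribbons and mesh particles.
# Emissive combines a curve-atlas color gradient with panning fire-noise, streak and wave layers, each sampled at three offset UVs for RGB fringing, with noise distortion.
# Opacity blends two scrolling motion-line layers with a noisy soft leading edge, then applies UV edge fades, particle alpha and near-camera fade.
# Depth fade is skipped on mobile.
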Material /Caretaker/FX/Melee/M_Caretaker_Melee_AirFlowTrail
graph {
 properties {
  bUsedWithMeshDeformer=true
  bUsedWithNiagaraMeshParticles=true
  bUsedWithNiagaraRibbons=true
  bUsedWithSkeletalMesh=true
  blendMode=BLEND_Translucent
  pixelDepthOffsetMode=PDOM_Legacy
  shadingModel=MSM_Unlit
  twoSided=true
  usedShadingModels=Unlit
 }
 nodes [
  n1 = ParticleColor()
  n2 = ScalarParameter() {parameterName=Use Particle Color}
  n3 = ComponentMask(None=TextureCoordinate()) {r=true}
  n4 = Saturate(None=Multiply(A=Power(Base=@n3), B=ScalarParameter() {defaultValue=10, parameterName=StartGradientMult}) {constB=100})
  n5 = NamedRerouteUsage()
  n6 = VectorParameter() {parameterName=Color Low, defaultValue=<rgba 0.0343 0 0.1878 1>}
  n7 = VectorParameter() {parameterName=Color High, defaultValue=<rgba 0.0908 0.0222 0.147 1>}
  n8 = ComponentMask(None=TextureCoordinate()) {r=true}
  n9 = CurveAtlasRowParameter(CurveTime=@n8) {
   atlas=/Game/Athena/LinearColorCurves/VFX_CurveAtlas
   curve=/Game/Athena/LinearColorCurves/Curve_Burst_Up_Slow_Down
   defaultValue=5
   parameterName=Global_Color
  }
  n10 = Multiply(A=@n9)
  n11 = Multiply(A=Desaturation(None=LinearInterpolate(A=@n6.RGB, B=@n7.RGB, Alpha=@n10) {constB=0}, Fraction=Constant() {r=-2})) {constB=4}
  n12 = Constant3Vector() {constant=<rgba 1 0.2917 0.2993 1>}
  n13 = Add(A=NamedRerouteUsage(), B=Add(A=NamedRerouteUsage(), B=Multiply(A=@n11, B=@n12)))
  n14 = Multiply(A=Desaturation(None=Add(A=@n5, B=@n13), Fraction=Constant() {r=-0.2})) {constB=0.2}
  MaterialFunctionCall(Emissive=Multiply(A=LinearInterpolate(B=@n1.RGB, Alpha=@n2) {constA=1}, B=Multiply(A=@n4, B=@n14) {constA=1})) {materialFunction=&MF_FX_EmissiveAdjustments}
  n15 = Constant() {r=1}
  n16 = ScalarParameter() {defaultValue=0.8, parameterName=U_Scale}
  n17 = Add(
   A=Multiply(A=TextureCoordinate() {vTiling=2.1}, B=AppendVector(A=@n15, B=@n16))
   B=DynamicParameter() {paramNames [Material Random, Param2, Param3, Param4]}.Material Random
  )
  n18 = TextureSample(UVs=Panner(
    Coordinate=MaterialFunctionCall(XY=@n17) {materialFunction=&Swizzle}.YX
    Time=NamedRerouteUsage()
   ) {speedY=0.05}) {samplerType=SAMPLERTYPE_Grayscale, texture=&LowResBlurredNoise}
  NamedRerouteDeclaration(None=Multiply(A=ComponentMask(None=@n18.RGB) {g=true, r=true}) {constB=0.05}) {name=Disturb_Noise, nodeColor=<rgba 0.7698 0 1 1>}
  n19 = TextureCoordinate() {uTiling=0.2, vTiling=0.84}
  n20 = AppendVector(A=Constant() {r=1}, B=ScalarParameter() {defaultValue=0.8, parameterName=U_Scale})
  n21 = MaterialFunctionCall(XY=Add(A=Multiply(A=@n19, B=@n20), B=DynamicParameter() {paramNames [Material Random, Param2, Param3, Param4]}.Material Random)) {materialFunction=&Swizzle}
  n22 = Constant() {r=0.004}
  n23 = NamedRerouteUsage()
  n24 = NamedRerouteUsage()
  n25 = Constant2Vector() {g=-0.1, r=-0.03}
  n26 = Panner(Coordinate=Add(A=Add(A=@n21.YX, B=@n22) {constB=0.005}, B=@n23), Time=@n24, Speed=@n25) {speedX=0.05, speedY=-0.1}
  n27 = Panner(Coordinate=Add(A=@n21.YX, B=@n23), Time=@n24, Speed=@n25) {speedX=0.05, speedY=-0.1}
  n28 = Panner(Coordinate=Add(A=Subtract(A=@n21.YX, B=@n22), B=@n23), Time=@n24, Speed=@n25) {speedX=0.05, speedY=-0.1}
  n29 = Power(Base=MaterialFunctionCall(
    X=TextureSample(UVs=@n26) {texture=/Game/Global/Textures/Default/Noise/T_FireNoise_01}.R
    Y=TextureSample(UVs=@n27) {texture=&T_FireNoise_01}.R
    Z=TextureSample(UVs=@n28) {texture=&T_FireNoise_01}.R
   ) {materialFunction=&MakeFloat3}.Result) {constExponent=1.5}
  n30 = Multiply(A=@n29, B=@n29)
  n31 = Constant() {r=-0.2}
  NamedRerouteDeclaration(None=Multiply(A=Desaturation(None=@n30, Fraction=@n31)) {constB=2}) {name=Panner_texture_01, nodeColor=<rgba 0 0.0132 1 1>}
  n32 = TextureCoordinate() {uTiling=0.5, vTiling=3.75}
  n33 = AppendVector(A=Constant() {r=1}, B=ScalarParameter() {defaultValue=0.8, parameterName=U_Scale})
  n34 = MaterialFunctionCall(XY=Add(A=Multiply(A=@n32, B=@n33), B=DynamicParameter() {paramNames [Material Random, Param2, Param3, Param4]}.Material Random)) {materialFunction=&Swizzle}
  n35 = Constant() {r=0.005}
  n36 = Add(A=@n34.YX, B=@n35) {constB=0.005}
  n37 = NamedRerouteUsage()
  n38 = Constant2Vector() {g=-0.1, r=0.03}
  n39 = TextureSample(UVs=Panner(Coordinate=Add(A=@n36, B=@n23), Time=@n37, Speed=@n38) {speedX=0.05, speedY=-0.06}) {texture=/Game/Effects/Fort_Effects/Textures/Energy/T_OutpostShockwaveStreaks_02}
  n40 = TextureSample(UVs=Panner(Coordinate=Add(A=@n34.YX, B=@n23), Time=@n37, Speed=@n38) {speedX=0.05, speedY=-0.06}) {texture=&T_OutpostShockwaveStreaks_02}
  n41 = Subtract(A=@n34.YX, B=@n35)
  n42 = TextureSample(UVs=Panner(Coordinate=Add(A=@n41, B=@n23), Time=@n37, Speed=@n38) {speedX=0.05, speedY=-0.06}) {texture=&T_OutpostShockwaveStreaks_02}
  n43 = Desaturation(
   None=Power(Base=MaterialFunctionCall(X=@n39.R, Y=@n40.R, Z=@n42.R) {materialFunction=&MakeFloat3}.Result) {constExponent=1.5}
   Fraction=Constant() {r=-1}
  )
  NamedRerouteDeclaration(None=Multiply(A=Constant3Vector() {constant=<rgba 1 0.8143 0.2855 1>}, B=Multiply(A=@n43) {constB=2})) {name=Panner_texture_02, nodeColor=<rgba 0 0.0132 1 1>}
  n44 = Constant() {r=1}
  n45 = ScalarParameter() {defaultValue=0.8, parameterName=U_Scale}
  n46 = MaterialFunctionCall(XY=Multiply(A=TextureCoordinate() {uTiling=0.3, vTiling=0.21}, B=AppendVector(A=@n44, B=@n45))) {materialFunction=&Swizzle}
  n47 = TextureCoordinate() {uTiling=1.5, vTiling=3.15}
  n48 = AppendVector(A=Constant() {r=1}, B=ScalarParameter() {defaultValue=0.8, parameterName=U_Scale})
  n49 = Panner(
   Coordinate=MaterialFunctionCall(XY=Multiply(A=@n47, B=@n48)) {materialFunction=&Swizzle}.YX
   Time=NamedRerouteUsage()
  ) {speedY=0.1}
  n50 = Multiply(A=ComponentMask(None=TextureSample(UVs=@n49) {samplerType=SAMPLERTYPE_Grayscale, texture=&LowResBlurredNoise}.RGB) {g=true, r=true}) {constB=0.06}
  n51 = Add(A=Add(A=@n46.YX, B=Constant() {r=0.007}) {constB=0.005}, B=@n50)
  n52 = Constant2Vector() {r=0.5}
  n53 = NamedRerouteUsage()
  n54 = TextureSample(UVs=Panner(Coordinate=Add(A=@n51, B=@n52), Time=@n53) {speedY=-0.06}) {texture=/Game/Characters/Player/Female/Medium/Bodies/F_MED_Buffet_Shard/FX/T_Wave}
  n55 = Add(A=@n46.YX, B=@n50)
  n56 = TextureSample(UVs=Panner(Coordinate=Add(A=@n55, B=@n52), Time=@n53) {speedY=-0.06}) {texture=&T_Wave}
  n57 = Add(A=@n50, B=Subtract(A=@n46.YX, B=Constant() {r=0.003}))
  n58 = TextureSample(UVs=Panner(Coordinate=Add(A=@n57, B=@n52), Time=@n53) {speedY=-0.06}) {texture=&T_Wave}
  n59 = Power(Base=Desaturation(
    None=MaterialFunctionCall(X=@n54.R, Y=@n56.B, Z=@n58.B) {materialFunction=&MakeFloat3}.Result
    Fraction=Constant()
   ))
  NamedRerouteDeclaration(None=Multiply(A=Multiply(A=@n59) {constB=20}, B=Constant3Vector() {constant=<rgba 1 0.6406 0.6406 0>})) {name=Panner_texture_03, nodeColor=<rgba 0 0.0132 1 1>}
  NamedRerouteDeclaration(None=Multiply(A=Time(), B=ScalarParameter() {defaultValue=1, parameterName=PanSpeedMult})) {name=PanSpeedMultiplier, nodeColor=<rgba 0.9922 0 1 1>}
  n60 = ScalarParameter() {defaultValue=0.5, group=Alpha, parameterName=Fine Motion Lines Opacity A, sortPriority=10}
  n61 = ScalarParameter() {defaultValue=1, group=Alpha, parameterName=Fine Motion Lines Opacity B, sortPriority=11}
  n62 = ScalarParameter() {defaultValue=2, group=Motion Lines, parameterName=Motion Lines A Intensity, sortPriority=3}
  n63 = VectorParameter() {group=Motion Lines, parameterName=Fine Motion Lines UV Scale A, sortPriority=5, defaultValue=<rgba 0.3 1.5 0 0>}
  n64 = TextureCoordinate()
  n65 = VectorParameter() {group=Motion Lines, parameterName=Fine Motion Lines Panner Speed A, sortPriority=4, defaultValue=<rgba 0 0.976 0 0>}
  n66 = Panner(Coordinate=Multiply(A=@n64, B=ComponentMask(None=@n63.RGB) {g=true, r=true}), Speed=ComponentMask(None=@n65.RGB) {g=true, r=true})
  n67 = ScalarParameter() {defaultValue=1, group=Motion Lines, parameterName=Motion Lines Swizzle UVs, sortPriority=12}
  n68 = TextureSampleParameter2D(UVs=LinearInterpolate(A=@n66, B=MaterialFunctionCall(XY=@n66) {materialFunction=&Swizzle}.YX, Alpha=@n67)) {
   group=Motion Lines
   parameterName=Motion Lines Texture
   samplerSource=SSM_Wrap_WorldGroupSettings
   texture=/Game/Characters/Player/Female/Medium/Bodies/F_MED_Buffet_Shard/FX/T_Flare_Ray
  }
  n69 = Power(
   Base=ChannelMaskParameter(None=@n68.RGBA) {group=Motion Lines, parameterName=Motion Lines Texture Mask A, sortPriority=1}.R
   Exp=ScalarParameter() {defaultValue=1, group=Motion Lines, parameterName=Motion Lines Texture A Contrast, sortPriority=2}
  ) {constExponent=1.5}
  n70 = VectorParameter() {group=Motion Lines, parameterName=Fine Motion Lines UV Scale B, sortPriority=11, defaultValue=<rgba 0.2 1.5 0 0>}
  n71 = VectorParameter() {group=Motion Lines, parameterName=Fine Motion Lines Panner Speed B, sortPriority=10, defaultValue=<rgba 0 -0.7616 0 0>}
  n72 = Panner(Coordinate=Multiply(A=@n64, B=ComponentMask(None=@n70.RGB) {g=true, r=true}), Speed=ComponentMask(None=@n71.RGB) {g=true, r=true}) {speedY=-0.7616}
  n73 = TextureSampleParameter2D(UVs=LinearInterpolate(A=@n72, B=MaterialFunctionCall(XY=@n72) {materialFunction=&Swizzle}.YX, Alpha=@n67)) {
   group=Motion Lines
   parameterName=Motion Lines Texture
   samplerSource=SSM_Wrap_WorldGroupSettings
   texture=&T_Flare_Ray
  }
  n74 = ChannelMaskParameter(None=@n73.RGBA) {group=Motion Lines, maskChannel=Green, parameterName=Motion Lines Texture Mask B, sortPriority=7, defaultValue=<rgba 0 1 0 0>}
  n75 = Power(Base=@n74.R, Exp=ScalarParameter() {defaultValue=1, group=Motion Lines, parameterName=Motion Lines Texture B Contrast, sortPriority=8})
  n76 = ScalarParameter() {defaultValue=5, group=Motion Lines, parameterName=Motion Lines B Intensity, sortPriority=9}
  n77 = Multiply(
   A=Add(A=Multiply(A=@n62, B=@n69), B=Multiply(A=@n75, B=@n76))
   B=ScalarParameter() {defaultValue=0.5, group=Motion Lines, parameterName=Combined Motion Lines Scale}
  )
  n78 = TextureCoordinate()
  n79 = VectorParameter() {group=Soft Leading Edge, parameterName=Soft Leading Edge UV Scale, sortPriority=6, defaultValue=<rgba 1 2 0 0>}
  n80 = ComponentMask(None=@n79.RGB) {g=true, r=true}
  n81 = VectorParameter() {group=Soft Leading Edge, parameterName=Soft Leading Edge Panner Speed, sortPriority=7, defaultValue=<rgba -0.65 0 0 0>}
  n82 = TextureSampleParameter2D(UVs=Panner(Coordinate=Multiply(A=@n78, B=@n80), Speed=ComponentMask(None=@n81.RGB) {g=true, r=true}) {bFractionalPart=true}) {
   group=Soft Leading Edge
   parameterName=Soft Leading Edge Texture
   samplerSource=SSM_Wrap_WorldGroupSettings
   sortPriority=4
   texture=&T_NoisePack_01
  }
  n83 = ChannelMaskParameter(None=@n82.RGBA) {group=Soft Leading Edge, maskChannel=Green, parameterName=Soft Leading Edge Texture Mask, sortPriority=5, defaultValue=<rgba 0 1 0 0>}
  n84 = Add(A=@n83.R, B=ScalarParameter() {defaultValue=0.25, group=Soft Leading Edge, parameterName=Bias, sortPriority=3})
  n85 = ScalarParameter() {defaultValue=1, group=Soft Leading Edge, parameterName=Scale, sortPriority=2}
  n86 = ComponentMask(None=TextureCoordinate()) {r=true}
  n87 = Power(
   Base=Add(A=Multiply(A=@n84, B=@n85), B=OneMinus(None=@n86))
   Exp=ScalarParameter() {defaultValue=-1, group=Soft Leading Edge, parameterName=Soft Leading Edge Contrast, sortPriority=1}
  ) {constExponent=6}
  n88 = ScalarParameter() {defaultValue=1, group=Soft Leading Edge, parameterName=Use Soft Leading Edge}
  n89 = ComponentMask(None=TextureCoordinate()) {g=true}
  n90 = Multiply(A=@n89, B=ScalarParameter() {defaultValue=2, group="Fade Top && Bottom Length Edges", parameterName=Top Length Fade Scale}) {constB=3}
  n91 = Multiply(
   A=OneMinus(None=@n89)
   B=ScalarParameter() {defaultValue=2, group="Fade Top && Bottom Length Edges", parameterName=Bottom Length Fade Scale, sortPriority=1}
  ) {constB=2}
  n92 = Multiply(
   A=LinearInterpolate(
    A=LinearInterpolate(A=@n60, B=@n61, Alpha=@n77) {constA=1}
    B=ScalarParameter() {defaultValue=0.8, group=Alpha, parameterName=Leading Edge Opacity, sortPriority=9}
    Alpha=Multiply(A=@n87, B=@n88)
   ) {constAlpha=0}
   B=Multiply(A=Saturate(None=@n90), B=Saturate(None=@n91))
  )
  n93 = ComponentMask(None=TextureCoordinate()) {r=true}
  n94 = Multiply(
   A=Add(A=@n93, B=ScalarParameter() {group="Fade Leading && Trailing Edges", parameterName=Leading Edge Fade Bias, sortPriority=2})
   B=ScalarParameter() {defaultValue=3, group="Fade Leading && Trailing Edges", parameterName=Leading Edge Fade Scale}
  ) {constB=3}
  n95 = Multiply(
   A=Add(A=OneMinus(None=@n93), B=ScalarParameter() {group="Fade Leading && Trailing Edges", parameterName=Trailing Edge Fade Bias, sortPriority=3})
   B=ScalarParameter() {defaultValue=0.25, group="Fade Leading && Trailing Edges", parameterName=Trailing Edge Fade Scale, sortPriority=1}
  ) {constB=2}
  n96 = Multiply(
   A=LinearInterpolate(B=Saturate(None=@n94), Alpha=ScalarParameter() {defaultValue=1, group=Alpha, parameterName=Fade Leading Edge, sortPriority=4}) {constA=1}
   B=LinearInterpolate(B=Saturate(None=@n95), Alpha=ScalarParameter() {defaultValue=1, group=Alpha, parameterName=Fade Trailing Edge, sortPriority=5}) {constA=1}
  )
  n97 = ScalarParameter() {defaultValue=64, group=Alpha, parameterName=NearFadeDistance, sortPriority=3}
  n98 = Multiply(
   A=ParticleColor().A
   B=Multiply(
    A=Multiply(A=@n92, B=@n96)
    B=MaterialFunctionCall(Distance=@n97) {materialFunction=&Near_Camera_Fade}.Result
   )
  )
  n99 = ScalarParameter() {defaultValue=50, group=Alpha, parameterName=FadeDistance, sortPriority=2}
  NamedRerouteDeclaration(None=ShadingPathSwitch(Default=DepthFade(Opacity=@n98, FadeDistance=@n99), Mobile=@n98)) {name=Opacity, nodeColor=<rgba 0.2031 0.5219 1 1>}
  NamedRerouteUsage()
  OneMinus(None=@n8)
 ]
}
