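# Niagara system for the Caretaker's melee charge-up: purple flares, darkness and right-arm electricity.
# Sprite emitters track Caretaker skeletal mesh bones or sockets; core and flicker flares intensify along a normalized system timeline.
# Anamorphic flare and darkness emitters start after a fraction of User.WarmUpTime, which also sets their loop duration.
# Electric ribbons use the custom BoneFlow module to twist randomized strands along four bone IDs, with spline-binormal facing and gradient color.
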NiagaraSystem /Caretaker/FX/Melee/NS_Caretaker_Melee_Charging
dependencies {
 usedDataInterfaces [
  &NiagaraDataInterfaceCurve
  &NiagaraDataInterfaceColorCurve
  &NiagaraDataInterfaceVector2DCurve
  &NiagaraDataInterfaceSkeletalMesh
 ]
 usedDynamicInputs [&Divide_Float]
 usedModules [
  &self:SystemSpawnScript.SystemScriptSource.SystemScriptGraph.NiagaraNodeAssignment_0.SetVariables_#1
  &SystemState
  &Timeline
 ]
}
emitters {
 Animorphic {
  emitterUpdateScript [
   &EmitterState(
    Delay First Loop Only=true
    Inactive Response=::NewEnumerator1
    Life Cycle Mode=::NewEnumerator1
    Loop Delay=&Multiply_Float(A=$User.WarmUpTime, B=0.375)
    Loop Duration=$User.WarmUpTime
    UseLoopDelay=true
   )
   &SpawnRate(SpawnRate=30, Use Spawn Probability=false)
  ]
  particleSpawnScript [
   &PartitionParticles()
   &InitializeParticle(
    Color=<rgba 2.4945 0.1507 5 1>
    Lifetime Max=0.08
    Lifetime Min=0.03
    Lifetime Mode=::NewEnumerator1
    Sprite Size=<2880 520>
    Sprite Size Mode=::NewEnumerator4
    UsePositionOffset=false
    Write Material Random=true
   )
   &CameraOffset(Camera Offset Amount=150)
  ]
  particleUpdateScript [
   &ParticleState()
   &ScaleSpriteSize()
   &ScaleColor(
    Scale Alpha=&FloatFromCurve(FloatCurve=DataInterface() {propertyValues {Curve {keys [<at 0 1 RCIM_Cubic>, {arriveTangent=-2.8468, interpMode=RCIM_Cubic, leaveTangent=-2.8468, tangentMode=RCTM_User, time=1, value=0}]}}})
    ScaleA=true
    ScaleRGB=true
   )
   &SkeletalMeshLocation(
    Bone / Socket Index=$Particles.PartitionParticles.Partition
    MaskSpawnGroup=false
    Mesh Sampling Type=::NewEnumerator5
    Skeletal Mesh=$System.CaretakerMesh
    Socket Sampling Mode=::NewEnumerator5
   )
  ]
  rendererClasses [&NiagaraSpriteRendererProperties]
  renderers [{rendererClass=&NiagaraSpriteRendererProperties}]
 }
 Building_Flicker_Flare {
  emitterUpdateScript [
   &EmitterState()
   &SpawnRate(SpawnRate=30, Use Spawn Probability=false)
  ]
  particleSpawnScript [
   &PartitionParticles()
   &InitializeParticle(
    Color=<rgba 0.9978 0.0603 2 0.5>
    Lifetime=&FloatFromCurve(FloatCurve=DataInterface() {propertyValues {Curve {keys [<at 0 0.15>,<at 0.8 0.05>]}}}, CurveIndex=$System.Timeline.NormalizedTime)
    Sprite Rotation Mode=::NewEnumerator2
    Sprite Size Mode=::NewEnumerator1
    Uniform Sprite Size Max=&FloatFromCurve(
     FloatCurve=DataInterface() {propertyValues {Curve {keys [{arriveTangent=579.03, interpMode=RCIM_Cubic, leaveTangent=579.03, tangentMode=RCTM_User, time=0, value=5}, <at 1 260 RCIM_Cubic>]}}}
     CurveIndex=$System.Timeline.NormalizedTime
     Scale Curve=4
    )
    Uniform Sprite Size Min=&FloatFromCurve(
     FloatCurve=DataInterface() {propertyValues {Curve {keys [{arriveTangent=367.8375, interpMode=RCIM_Cubic, leaveTangent=367.8375, tangentMode=RCTM_User, time=0, value=1}, <at 1 170 RCIM_Cubic>]}}}
     CurveIndex=$System.Timeline.NormalizedTime
     Scale Curve=4
    )
    UsePositionOffset=false
    Write Material Random=true
   )
   &CameraOffset(Camera Offset Amount=175)
  ]
  particleUpdateScript [
   &ParticleState()
   &ScaleColor(
    Scale Alpha=&FloatFromCurve(FloatCurve=DataInterface() {propertyValues {Curve {keys [<at 0 1 RCIM_Cubic>, {arriveTangent=-2.9061, interpMode=RCIM_Cubic, leaveTangent=-2.9061, tangentMode=RCTM_User, time=1, value=0}]}}})
    ScaleA=true
    ScaleRGB=true
   )
   &SkeletalMeshLocation(
    Bone / Socket Index=$Particles.PartitionParticles.Partition
    MaskSpawnGroup=false
    Mesh Sampling Type=::NewEnumerator5
    Skeletal Mesh=$System.CaretakerMesh
    Socket Sampling Mode=::NewEnumerator5
   )
  ]
  rendererClasses [&NiagaraSpriteRendererProperties]
  renderers [{rendererClass=&NiagaraSpriteRendererProperties}]
 }
 Core_Flare {
  emitterUpdateScript [
   &EmitterState()
   &SpawnRate(SpawnRate=40, Use Spawn Probability=false)
  ]
  particleSpawnScript [
   &PartitionParticles()
   &InitializeParticle(
    Color=&MakeLinearColorFromVectorAndFloat(
     "Vector (RGB)"=&MakeVectorFromLinearColorRGB(Color=<rgba 0.7484 0.0452 1.5 1>)
     "Float (Alpha)"=&FloatFromCurve(FloatCurve=DataInterface() {propertyValues {Curve {keys [<at 0 0>,<at 1 0.8>]}}}, CurveIndex=$System.Timeline.NormalizedTime)
    )
    Lifetime Max=0.15
    Lifetime Min=0.1
    Lifetime Mode=::NewEnumerator1
    Sprite Size Mode=::NewEnumerator1
    Uniform Sprite Size Max=&FloatFromCurve(
     FloatCurve=DataInterface() {propertyValues {Curve {keys [{arriveTangent=579.03, interpMode=RCIM_Cubic, time=0, value=10}, <at 1 150 RCIM_Cubic>]}}}
     CurveIndex=$System.Timeline.NormalizedTime
     Scale Curve=10
    )
    Uniform Sprite Size Min=&FloatFromCurve(
     FloatCurve=DataInterface() {propertyValues {Curve {keys [{arriveTangent=298.8282, interpMode=RCIM_Cubic, time=0, value=10}, <at 1 110 RCIM_Cubic>]}}}
     CurveIndex=$System.Timeline.NormalizedTime
     Scale Curve=10
    )
    UsePositionOffset=false
    Write Material Random=true
   )
   &CameraOffset(Camera Offset Amount=200)
   &DynamicMaterialParameters(
    Index 0 Param 1=0.2
    Index 0 Param 2=1
    Index 0 Param 3=1
    Index 0 Param 4=1
    Module.Param0WriteEnabled=true
    Module.Param1WriteEnabled=true
    Module.Param2WriteEnabled=true
    Module.Param3WriteEnabled=true
   )
  ]
  particleUpdateScript [
   &ParticleState()
   &ScaleColor(
    Scale Alpha=&FloatFromCurve(FloatCurve=DataInterface() {propertyValues {Curve {keys [<at 0 1>,<at 1 0>]}}})
    Scale RGB=<4 4 4>
    ScaleA=true
    ScaleRGB=true
   )
   /SetVariables(
    Particles.Flare B Color:LinearColor=&MultiplyLinearColor(
     Linear Color=<rgba 0.7484 0.0452 1.5 1>
     Scale Factor=&MakeFloat4FromVectorAndFloat(
      "Vector (RGB)"=<4 4 4>
      "Float (Alpha)"=&FloatFromCurve(FloatCurve=DataInterface() {propertyValues {Curve {keys [<at 0 1>,<at 1 0>]}}})
     )
    )
    Particles.Flare B Size:Vector2f=&Multiply_Vector2D(
     A=&MakeVector2D(
      X=&RandomRangeFloat(
       Minimum=&FloatFromCurve(FloatCurve=DataInterface() {propertyValues {Curve {keys [<at 0 9>,<at 1 80>]}}}, CurveIndex=$System.Timeline.NormalizedTime)
       Maximum=&FloatFromCurve(FloatCurve=DataInterface() {propertyValues {Curve {keys [<at 0 17>,<at 1 180>]}}}, CurveIndex=$System.Timeline.NormalizedTime)
      )
      Y=&RandomRangeFloat(
       Minimum=&FloatFromCurve(FloatCurve=DataInterface() {propertyValues {Curve {keys [<at 0 3>,<at 1 4>]}}}, CurveIndex=$System.Timeline.NormalizedTime)
       Maximum=&FloatFromCurve(FloatCurve=DataInterface() {propertyValues {Curve {keys [<at 0 18>,<at 1 30>]}}}, CurveIndex=$System.Timeline.NormalizedTime)
      )
     )
     B=&Vector2DFromCurve(
      Vector2Curve=DataInterface() {propertyValues {
        XCurve {keys [
          <at 0 0.265 RCIM_Cubic>
          {arriveTangent=-0.1232, interpMode=RCIM_Cubic, leaveTangent=-0.1232, tangentMode=RCTM_User, time=0.17, value=0.61}
          {arriveTangent=1.072, interpMode=RCIM_Cubic, leaveTangent=1.072, tangentMode=RCTM_User, time=0.7, value=0.77}
          <at 1 1 RCIM_Cubic>
         ]}
        YCurve {keys [<at 0.7 1 RCIM_Cubic>,<at 1 0 RCIM_Cubic>]}
       }}
      ScaleCurve=<1 1>
     )
    )
   ) {isSetParameters=true}
   &ScaleSpriteSize(Uniform Curve Sprite Scale=DataInterface() {propertyValues {Curve {keys [
        <at 0 0.265 RCIM_Cubic>
        {arriveTangent=-0.0902, interpMode=RCIM_Cubic, leaveTangent=-0.0902, tangentMode=RCTM_User, time=0.17, value=0.61}
        {arriveTangent=1.2869, interpMode=RCIM_Cubic, leaveTangent=1.2869, tangentMode=RCTM_User, time=0.7, value=0.77}
        <at 1 1 RCIM_Cubic>
       ]}}})
   &SkeletalMeshLocation(
    Bone / Socket Index=$Particles.PartitionParticles.Partition
    MaskSpawnGroup=false
    Mesh Sampling Type=::NewEnumerator5
    Skeletal Mesh=$System.CaretakerMesh
    Socket Sampling Mode=::NewEnumerator5
   )
  ]
  rendererClasses [&NiagaraSpriteRendererProperties]
  renderers [
   {rendererClass=&NiagaraSpriteRendererProperties}
   {rendererClass=&NiagaraSpriteRendererProperties, rendererIndex=1}
  ]
 }
 Darkness {
  emitterUpdateScript [
   &EmitterState(
    Delay First Loop Only=true
    Inactive Response=::NewEnumerator1
    Life Cycle Mode=::NewEnumerator1
    Loop Delay=&Multiply_Float(A=$User.WarmUpTime, B=0.375)
    Loop Duration=$User.WarmUpTime
    UseLoopDelay=true
   )
   &SpawnRate(SpawnRate=5, Use Spawn Probability=false)
   &SpawnBurst_Instantaneous(Spawn Count=4, Use Spawn Probability=false, UseLoopCountLimit=true)
  ]
  particleSpawnScript [
   &PartitionParticles()
   &InitializeParticle(
    Color=<rgba 0.0107 0.0036 0.02 1>
    Lifetime=1.6
    Sprite Size Mode=::NewEnumerator3
    Uniform Sprite Size=600
    UsePositionOffset=false
    Write Material Random=true
   )
   &DynamicMaterialParameters(
    Index 0 Param 1=3
    Index 0 Param 2=1
    Index 0 Param 3=1
    Index 0 Param 4=1
    Module.Param0WriteEnabled=true
    Module.Param1WriteEnabled=true
    Module.Param2WriteEnabled=true
    Module.Param3WriteEnabled=true
   )
   &CameraOffset(Camera Offset Amount=150)
  ]
  particleUpdateScript [
   &ParticleState()
   &ScaleColor(
    Scale Alpha=&FloatFromCurve(FloatCurve=DataInterface() {propertyValues {Curve {keys [
         <at 0 0 RCIM_Cubic>
         {arriveTangentWeight=0.2592, interpMode=RCIM_Cubic, leaveTangentWeight=0.1296, tangentMode=RCTM_User, time=0.4, value=1}
         {arriveTangentWeight=0.1296, interpMode=RCIM_Cubic, leaveTangentWeight=0.2592, tangentMode=RCTM_User, time=0.6, value=1}
         <at 1 0 RCIM_Cubic>
        ]}}})
    ScaleA=true
    ScaleRGB=true
   )
   &SkeletalMeshLocation(
    Bone / Socket Index=$Particles.PartitionParticles.Partition
    MaskSpawnGroup=false
    Mesh Sampling Type=::NewEnumerator5
    Skeletal Mesh=$System.CaretakerMesh
    Socket Sampling Mode=::NewEnumerator5
   )
  ]
  rendererClasses [&NiagaraSpriteRendererProperties]
  renderers [{rendererClass=&NiagaraSpriteRendererProperties}]
 }
 Right_Arm_Electric {
  emitterSpawnScript [
   /SetVariables(Emitter.RandomSeed:NiagaraInt32=/Niagara/DynamicInputs/Random/RandomInt(Randomness Mode:ENiagaraRandomnessMode=::NewEnumerator0, Random Seed:NiagaraInt32=0, Fixed Random Seed:NiagaraBool=false)) {isSetParameters=true}
   /SetVariables(Emitter.ParticlesPerGroup=20) {isSetParameters=true}
  ]
  emitterUpdateScript [
   /SetVariables(Emitter.NumberOfGroups:NiagaraInt32=&RandomRangeInt(Minimum=2, Maximum=6, Evaluation Type=::NewEnumerator1)) {isSetParameters=true}
   &EmitterState()
   &SpawnRate(
    SpawnRate=&Multiply_Float(
     A=&MakeFloatFromInt(Int=&Multiply_Int(A=$Emitter.ParticlesPerGroup, B=$Emitter.NumberOfGroups))
     B=3
    )
    Use Spawn Probability=false
   )
  ]
  particleSpawnScript [
   /SetVariables(Particles.LoopCount:NiagaraInt32=$Emitter.LoopCount) {isSetParameters=true}
   &FN_SetSpawnGroup(
    Inclusivity=true
    Normalize Spawn Group=true
    Particles Per Group:NiagaraInt32=$Emitter.ParticlesPerGroup
    Pattern=::NewEnumerator0
    Set Ribbon ID and Ribbon Link Order=true
   )
   &PartitionParticles(
    Particles Per Partition:NiagaraInt32=$Emitter.ParticlesPerGroup
    Partition Count Mode:ENiagara_InfiniteLoopDuration=::NewEnumerator1
    Partition Mode:ENiagara_PartitionMode=::NewEnumerator1
    Write Partition to Ribbon ID=::NewEnumerator0
   )
   &InitializeRibbon(
    Color=<rgba 30 30 30 1>
    Lifetime=&RandomRangeFloat(Minimum=0.1, Maximum=0.2)
    Module.Write Color=true
    Module.Write Lifetime=true
    Module.Write Mass=true
    Module.Write Material Random=true
    Module.Write Position=false
    Module.Write RibbonTwist=false
    Module.Write RibbonWidth=true
    Ribbon Width=20
   )
  ]
  particleUpdateScript [
   &ParticleState()
   /Game/Effects/Niagara/Modules/BoneFlow/FN_BoneFlow_v4/FN_BoneFlow(
    Angle:NiagaraFloat=&Add_Float(
     A=&Multiply_Float(
      A=$Particles.PartitionParticles.Normalized Index In Partition
      B=/Game/Effects/Niagara/DynamicInputs/Random/FN_SeededRandomRangeFloat(
       Minimum:NiagaraFloat=-3
       Maximum:NiagaraFloat=3
       Random Seed:NiagaraInt32=&Add_Integer(A=$Particles.LoopCount, B=$Particles.SpawnGroup)
      )
     )
     B=&Multiply_Float(A=$Particles.Age, B=0.2)
    )
    BoneID:NiagaraInt32=&ModuloInt(A=$Particles.PartitionParticles.Partition, B=4)
    Height:NiagaraFloat=$Particles.PartitionParticles.Normalized Index In Partition
    Interpolate Mesh Sampling=false
    Radius:NiagaraFloat=&RandomRangeFloat(
     Minimum=&FloatFromCurve(
      FloatCurve=DataInterface() {propertyValues {Curve {keys [
          <at 0 0 RCIM_Cubic>
          {arriveTangent=4.1777, interpMode=RCIM_Cubic, leaveTangent=4.1777, tangentMode=RCTM_User, time=0.2, value=10}
          {arriveTangent=-0.6996, interpMode=RCIM_Cubic, leaveTangent=-0.6996, tangentMode=RCTM_User, time=0.8, value=10}
          <at 1 0 RCIM_Cubic>
         ]}}}
      CurveIndex=$Particles.PartitionParticles.Normalized Index In Partition
      Scale Curve=5
     )
     Maximum=&FloatFromCurve(
      FloatCurve=DataInterface() {propertyValues {Curve {keys [
          <at 0 0 RCIM_Cubic>
          {arriveTangent=0.5753, interpMode=RCIM_Cubic, leaveTangent=0.5753, tangentMode=RCTM_User, time=0.2, value=2}
          {arriveTangent=0.7345, interpMode=RCIM_Cubic, leaveTangent=0.7345, tangentMode=RCTM_User, time=0.8, value=2}
          <at 1 0 RCIM_Cubic>
         ]}}}
      CurveIndex=$Particles.PartitionParticles.Normalized Index In Partition
      Scale Curve=5
     )
    )
    SkeletalMesh:NiagaraDataInterfaceSkeletalMesh=$System.CaretakerMesh
    Write Position Attribute=true
   ) {module=FN_BoneFlow001}
   &ScaleRibbonWidth(Ribbon Width Scale=&Multiply_Float(
     A=&FloatFromCurve(FloatCurve=DataInterface() {propertyValues {Curve {keys [<at 0 1 RCIM_Cubic>,<at 0.2 3 RCIM_Cubic>,<at 0.5 1 RCIM_Cubic>]}}})
     B=&FloatFromCurve(
      FloatCurve=DataInterface() {propertyValues {Curve {keys [<at 0 0.5 RCIM_Cubic>,<at 0.4995 1.2 RCIM_Cubic>,<at 1 0.5 RCIM_Cubic>]}}}
      CurveIndex=$Particles.PartitionParticles.Normalized Index In Partition
      Scale Curve=2
     )
    ))
   /SetVariables(Particles.RibbonFacing=$Output.FN_BoneFlow001.SplineBinormal) {isSetParameters=true}
   &ScaleColor(
    Scale RGB=&MakeVectorFromLinearColorRGB(Color=&ColorFromCurve(
      ColorCurve=DataInterface() {propertyValues {
        AlphaCurve {keys []}
        BlueCurve {keys [<at 0 0.013 RCIM_Cubic>, {arriveTangent=2.7442, interpMode=RCIM_Cubic, leaveTangent=2.7442, tangentMode=RCTM_User, time=1, value=1}]}
        GreenCurve {keys [<at 0 0 RCIM_Cubic>,<at 1 0.0301 RCIM_Cubic>]}
        RedCurve {keys [<at 0 0.147 RCIM_Cubic>, {arriveTangent=1.2723, interpMode=RCIM_Cubic, leaveTangent=1.2723, tangentMode=RCTM_User, time=1, value=0.4989}]}
       }}
      CurveIndex=$Particles.PartitionParticles.Normalized Index In Partition
     ))
    ScaleA=true
    ScaleRGB=true
   )
  ]
  rendererClasses [&NiagaraRibbonRendererProperties]
  renderers [{rendererClass=&NiagaraRibbonRendererProperties}]
 }
}
summary {
 systemName=NS_Caretaker_Melee_Charging
 emitters [
  {emitterName=Right_Arm_Electric, rendererClasses [&NiagaraRibbonRendererProperties]}
  {emitterName=Darkness, rendererClasses [&NiagaraSpriteRendererProperties]}
  {emitterName=Building_Flicker_Flare, rendererClasses [&NiagaraSpriteRendererProperties]}
  {emitterName=Animorphic, rendererClasses [&NiagaraSpriteRendererProperties]}
  {emitterName=Core_Flare, rendererClasses [&NiagaraSpriteRendererProperties]}
 ]
 userVariables [User.WarmUpTime {type=NiagaraFloat, defaultValue=2, description=""}]
}
